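# UI-domain material for the Sleek item shop background, layering flower artwork over a base texture.
# Corner flowers use scaled, offset UI UVs and blend by texture alpha within a sine-based band mask.
# Two more flowers are positioned by parameters, spun in opposite directions by time-driven Rotators, then blended using their own RGB as alpha.
# Separate orange-tint branches, using a vertical edge gradient and a fade texture, do not feed the flower composite.
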Material /ItemShop/UI/Art/Sleek/M_UI_SleekItemShopBG
graph {
 properties {bUsedWithStaticMesh=false, materialDomain=MD_UI, pixelDepthOffsetMode=PDOM_Legacy, usedShadingModels=Default Lit}
 nodes [
  LinearInterpolate(
   B=Constant3Vector() {constant=<rgba 0.4896 0.1036 0.007 0>}
   Alpha=Multiply(B=TextureSample() {texture=/ItemShop/UI/Art/Sleek/Texture/T_UI_SleekBG_Fade}.RGB) {constA=1.0004}
  )
  n1 = TextureSampleParameter2D(UVs=NamedRerouteUsage()) {parameterName=SleekTexture, texture=/ItemShop/UI/Art/Sleek/Texture/T_UI_Sleek}
  n2 = MaterialFunctionCall() {materialFunction=&GetUserInterfaceUV}
  n3 = MaterialFunctionCall(Float2=@n2.Pixel Size) {materialFunction=&BreakOutFloat2Components}
  n4 = AppendVector(A=Divide(A=@n3.R, B=@n3.R), B=Constant() {r=1})
  n5 = Add(
   A=Multiply(A=Multiply(A=@n4, B=@n2.Normalized UV), B=ScalarParameter() {defaultValue=1.1724, parameterName=Scale})
   B=AppendVector(A=ScalarParameter() {defaultValue=0.534, parameterName=X offset}, B=ScalarParameter() {defaultValue=-0.352, parameterName=Y Offset})
  )
  n6 = TextureSample(UVs=NamedRerouteDeclaration(None=@n5) {name=TextureCoordinates2, nodeColor=<rgba 0 0.6953 1 1>}) {texture=/ItemShop/UI/Art/Sleek/Texture/T_UI_CornerFlowers}
  n7 = Sine(None=Add(A=@n5, B=ScalarParameter() {defaultValue=-1.5118, parameterName=Offset})) {period=4}
  n8 = Subtract(A=Abs(None=ComponentMask(None=@n7) {r=true}), B=ScalarParameter() {defaultValue=0.698, parameterName=MaskRange}) {constB=0.35}
  n9 = ScalarParameter() {defaultValue=1139.6055, parameterName=MaskEdge}
  n10 = LinearInterpolate(B=@n6.A, Alpha=Saturate(None=Multiply(A=@n8, B=@n9) {constB=2}))
  n11 = VectorParameter() {parameterName=Flower3Scale, channelNames {a="", b="", g=Height, r=Width}, defaultValue=<rgba 0.236 0.42 1 1>}
  n12 = VectorParameter() {parameterName=Flower3Position, channelNames {a="", b="", g=Y, r=X}, defaultValue=<rgba 0.1037 0.2885 1 1>}
  n13 = Time()
  n14 = ScalarParameter() {defaultValue=-0.5, parameterName=TimeOffsetTop}
  n15 = TextureSample(UVs=Rotator(
    Coordinate=MaterialFunctionCall!(Texture Scale=@n11, UVs, Offset XY Texturesize Z=@n12) {materialFunction=/Engine/Functions/Engine_MaterialFunctions01/Texturing/ScaleUVsAroundPoint}.UVs
    Time=Multiply(A=@n13, B=@n14)
   ) {speed=0.13}) {samplerSource=SSM_Clamp_WorldGroupSettings, texture=/ItemShop/UI/Art/Sleek/Texture/T_UI_Flower_3}
  n16 = VectorParameter() {parameterName=Flower4Scale, channelNames {a="", b="", g=Height, r=Width}, defaultValue=<rgba 0.252 0.404 1 1>}
  n17 = VectorParameter() {parameterName=Flower4Position, channelNames {a="", b="", g=Y, r=X}, defaultValue=<rgba 0.9258 0.7707 1 1>}
  n18 = Time()
  n19 = ScalarParameter() {defaultValue=0.75, parameterName=TimeOffsetBottom}
  n20 = TextureSample(UVs=Rotator(
    Coordinate=MaterialFunctionCall(Texture Scale=@n16, Offset XY Texturesize Z=@n17) {materialFunction=&ScaleUVsAroundPoint}.UVs
    Time=Multiply(A=@n18, B=@n19)
   ) {speed=0.13}) {samplerSource=SSM_Clamp_WorldGroupSettings, texture=/ItemShop/UI/Art/Sleek/Texture/T_UI_Flower_4}
  LinearInterpolate(
   A=LinearInterpolate(A=LinearInterpolate(A=@n1.RGB, B=@n6.RGB, Alpha=@n10), B=@n15.RGB, Alpha=@n15.RGB) {constAlpha=1}
   B=@n20.RGB
   Alpha=@n20.RGB
  )
  n21 = Abs(None=ConstantBiasScale(None=TextureCoordinate()) {bias=-0.5, scale=2})
  n22 = Multiply(
   A=Subtract(A=ComponentMask(None=@n21) {g=true}, B=ScalarParameter() {defaultValue=0.7349, parameterName=Subtract})
   B=ScalarParameter() {defaultValue=1.4621, parameterName=Falloff}
  )
  LinearInterpolate(
   A=@n1.RGB
   B=Constant3Vector() {constant=<rgba 0.5625 0.121 0 1>}
   Alpha=Multiply(A=Saturate(None=@n22), B=ScalarParameter() {defaultValue=2.5483, parameterName=Intensity})
  )
  MaterialFunctionCall() {materialFunction=&ScreenResolution}
  n23 = MaterialFunctionCall() {materialFunction=&GetUserInterfaceUV}
  n24 = MaterialFunctionCall(Float2=@n23.Pixel Size) {materialFunction=&BreakOutFloat2Components}
  n25 = AppendVector(A=Constant() {r=1}, B=Divide(A=@n24.G, B=@n24.G))
  NamedRerouteDeclaration(None=Multiply(A=Multiply(A=@n25, B=@n23.Normalized UV), B=Constant2Vector() {g=2, r=1})) {name=TextureCoordinates, nodeColor=<rgba 0 0.6953 1 1>}
 ]
}
